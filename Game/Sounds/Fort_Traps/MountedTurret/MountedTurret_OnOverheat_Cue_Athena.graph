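# Sound cue for the mounted turret trap's overheat sound in Athena.
# A single ballistic overheat wave runs through a random pitch modulator and a short fixed delay, with cue-level volume and pitch lowered.
# It uses close-range weapon projectile attenuation and the third-person ranged weapon fire sound class.
# Its own concurrency override prevents new instances from starting once the limit is reached.
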SoundCue /Game/Sounds/Fort_Traps/MountedTurret/MountedTurret_OnOverheat_Cue_Athena
root=Delay_0
nodes [
 {id=Delay_0, children [Modulator_0], properties {delayMax=0.3, delayMin=0.3}}
 {id=Modulator_0, children [WavePlayer_0], properties {pitchMax=1.2, pitchMin=0.9}}
 {id=WavePlayer_0, properties {soundWaveAssetPtr=&/Turret_Ballistic_Overheat_04}}
]
properties {
 attenuationSettings=&Fort_Weapon_Projectile_Close_Athena_Att
 bOverrideConcurrency=true
 maxDistance=4950
 pitchMultiplier=0.9
 soundClassObject=&Fort_WeaponRanged_Fire_3P_PSM
 volumeMultiplier=0.4
 attenuationOverrides {distanceAlgorithm=NaturalSound, falloffDistance=4000, attenuationShapeExtents {x=1000}}
 concurrencyOverrides {resolutionRule=PreventNew}
}
