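# Sound cue for the mounted turret's overheat loop in Athena, layering a shaking rattle with a warning tone.
# A mixer blends two looping waves, each behind a continuous modulator driven by the Overheat parameter, with the rattle weighted louder.
# Overheat scales both volumes, the rattle peaking before maximum heat, and also raises the warning's pitch.
# Attenuation uses natural-sound falloff, and overridden concurrency prevents new instances at the limit.
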SoundCue /Game/Sounds/Fort_Traps/MountedTurret/MountedTurret_OverheatLoop_Cue_Athena
root=Mixer_0
nodes [
 {id=Mixer_0, children [ModulatorContinuous_0, ModulatorContinuous_1], properties {inputVolume [1.25,1]}}
 {
  id=ModulatorContinuous_0
  children [WavePlayer_1]
  properties {
   pitchModulationParams {default=0, minOutput=1, parameterName=Overheat}
   volumeModulationParams {default=0, maxInput=0.7, parameterName=Overheat}
  }
 }
 {
  id=ModulatorContinuous_1
  children [WavePlayer_0]
  properties {
   pitchModulationParams {default=0, maxOutput=1.15, minOutput=0.85, parameterName=Overheat}
   volumeModulationParams {default=0, parameterName=Overheat}
  }
 }
 {id=WavePlayer_0, properties {bLooping=true, soundWaveAssetPtr=&/MountedTurret_OverheatWarning_Loop}}
 {
  id=WavePlayer_1
  properties {bLooping=true, soundWaveAssetPtr=&/Turret_Ballistic_OverheatLoop_Shaking_01}
 }
]
properties {
 attenuationSettings=&Fort_Weapon_Projectile_Close_Athena_Att
 bOverrideConcurrency=true
 maxDistance=4950
 soundClassObject=/Game/Sounds/FortSoundClassesAndModes/Fort_WeaponRanged_LowAmmo_Critical_PSM
 volumeMultiplier=0.7
 attenuationOverrides {distanceAlgorithm=NaturalSound, falloffDistance=4000, attenuationShapeExtents {x=1000}}
 concurrencyOverrides {resolutionRule=PreventNew}
}
