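# Sound cue for the appearance of the ballistic mounted turret trap.
# A single wave player feeds the output directly, playing one appear sound with no randomization or modulation.
# It uses the shared trap attenuation settings and the trap placement sound class, at slightly reduced volume.
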SoundCue /Game/Sounds/Fort_Traps/MountedTurret/Turret_Ballistic_Appear_Cue
root=WavePlayer_1
nodes [{id=WavePlayer_1, properties {soundWaveAssetPtr=&/Turret_Ballistic_Appear_01}}]
properties {
 attenuationSettings=&Fort_Traps_Att
 maxDistance=6000
 soundClassObject=/Game/Sounds/FortSoundClassesAndModes/Fort_Traps_Placement
 volumeMultiplier=0.8
}
