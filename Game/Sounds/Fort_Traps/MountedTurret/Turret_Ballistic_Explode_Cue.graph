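# Sound cue for the explosion of the ballistic mounted turret trap.
# A random node picks one of two explosion waves with equal weight, and a modulator node above it varies the result.
# It plays through the shared trap attenuation and the medium explosives sound class at reduced volume.
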SoundCue /Game/Sounds/Fort_Traps/MountedTurret/Turret_Ballistic_Explode_Cue
root=Modulator_5
nodes [
 {id=Modulator_5, children [Random_2]}
 {id=Random_2, children [WavePlayer_8, WavePlayer_11], properties {weights [1,1]}}
 {id=WavePlayer_11, properties {soundWaveAssetPtr=&/Turret_Ballistic_Explode_02}}
 {id=WavePlayer_8, properties {soundWaveAssetPtr=&/Turret_Ballistic_Explode_01}}
]
properties {
 attenuationSettings=&Fort_Traps_Att
 maxDistance=6000
 soundClassObject=/Game/Sounds/FortSoundClassesAndModes/Fort_Explosives_Medium
 volumeMultiplier=0.384
}
