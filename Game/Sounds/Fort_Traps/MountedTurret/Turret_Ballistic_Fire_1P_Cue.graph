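# First-person firing sound cue for the mounted ballistic turret trap.
# A mixer layers the gunshot with two quieter, delayed military sniper casing sounds, one pitched up and one pitched down, under slight overall volume variation.
# A quality switch plays a random pick of four shots in its first two slots and one fixed shot in the other two.
# Playback uses first-person weapon attenuation, sound class and gunfire concurrency.
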SoundCue /Game/Sounds/Fort_Traps/MountedTurret/Turret_Ballistic_Fire_1P_Cue
root=Modulator_0
nodes [
 {id=Delay_0, children [Random_4], properties {delayMax=0.8, delayMin=0.7}}
 {id=Delay_1, children [Random_6], properties {delayMax=0.7, delayMin=0.6}}
 {id=Mixer_0, children [QualityLevel_0, Modulator_2, Modulator_4], properties {inputVolume [1,0.2,0.2]}}
 {id=Modulator_0, children [Mixer_0], properties {volumeMax=1, volumeMin=0.9}}
 {id=Modulator_2, children [Delay_0], properties {pitchMax=1.1, pitchMin=1}}
 {id=Modulator_4, children [Delay_1], properties {pitchMax=0.9, pitchMin=0.8}}
 {id=QualityLevel_0, children [Random_5, Random_5, WavePlayer_24, WavePlayer_24]}
 {id=Random_4, children [WavePlayer_16, WavePlayer_17, WavePlayer_18], properties {weights [1,1,1]}}
 {id=Random_5, children [WavePlayer_20, WavePlayer_21, WavePlayer_22, WavePlayer_23], properties {weights [1,1,1,1]}}
 {id=Random_6, children [WavePlayer_25, WavePlayer_26, WavePlayer_27], properties {weights [1,1,1]}}
 {id=WavePlayer_16, properties {soundWaveAssetPtr=/Game/Sounds/Fort_Weapons/SniperRifles/MilitarySniper/MilitarySniper_Casing_01}}
 {id=WavePlayer_17, properties {soundWaveAssetPtr=/Game/Sounds/Fort_Weapons/SniperRifles/MilitarySniper/MilitarySniper_Casing_02}}
 {id=WavePlayer_18, properties {soundWaveAssetPtr=/Game/Sounds/Fort_Weapons/SniperRifles/MilitarySniper/MilitarySniper_Casing_03}}
 {id=WavePlayer_20, properties {soundWaveAssetPtr=&/Turret_Ballistic_Fire_01}}
 {id=WavePlayer_21, properties {soundWaveAssetPtr=&/Turret_Ballistic_Fire_02}}
 {id=WavePlayer_22, properties {soundWaveAssetPtr=&/Turret_Ballistic_Fire_03}}
 {id=WavePlayer_23, properties {soundWaveAssetPtr=&/Turret_Ballistic_Fire_04}}
 {id=WavePlayer_24, properties {soundWaveAssetPtr=&Turret_Ballistic_Fire_01}}
 {id=WavePlayer_25, properties {soundWaveAssetPtr=&MilitarySniper_Casing_01}}
 {id=WavePlayer_26, properties {soundWaveAssetPtr=&MilitarySniper_Casing_02}}
 {id=WavePlayer_27, properties {soundWaveAssetPtr=&MilitarySniper_Casing_03}}
]
properties {
 attenuationSettings=&Fort_Weapon_Projectile_FirstPerson_Athena_Att
 maxDistance=2097152
 soundClassObject=&Fort_WeaponRanged_Fire_1P_PSM
 volumeMultiplier=0.85
 attenuationOverrides {distanceAlgorithm=NaturalSound, falloffDistance=4000, attenuationShapeExtents {x=1000}}
 concurrencyOverrides {maxCount=5}
 concurrencySet [&Weapon_Gunfire_FirstPerson_Conc]
}
