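# Firing sound cue for the ballistic mounted turret trap, mixing close, medium and distant layers with separate attenuations and sound classes.
# The close layer adds two quieter, delayed, pitch-shifted random casing sounds borrowed from the military sniper rifle.
# Quality-level nodes pick either random fire variants or a single fixed wave for each layer.
# The cue overrides concurrency to cap simultaneous instances.
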SoundCue /Game/Sounds/Fort_Traps/MountedTurret/Turret_Ballistic_Fire_Cue
root=Mixer_15
nodes [
 {
  id=Attenuation_80
  children [Modulator_0]
  properties {attenuationSettings=/Game/Sounds/Attenuation/Fort_Weapon_Projectile_Close_Athena_Att_NonSpac}
 }
 {
  id=Attenuation_81
  children [Modulator_13]
  properties {attenuationSettings=&Fort_Weapon_Projectile_Medium_Mono_Athena_Att}
 }
 {
  id=Attenuation_82
  children [Modulator_1]
  properties {attenuationSettings=&Fort_Weapon_Projectile_Distant_Mono_Athena_Att}
 }
 {id=Delay_0, children [Random_4], properties {delayMax=0.8, delayMin=0.7}}
 {id=Delay_1, children [Random_6], properties {delayMax=0.7, delayMin=0.6}}
 {id=Mixer_0, children [QualityLevel_0, Modulator_2, Modulator_4], properties {inputVolume [1,0.2,0.2]}}
 {id=Mixer_1, children [SoundClass_56, SoundClass_0], properties {inputVolume [1,1]}}
 {id=Mixer_15, children [SoundClass_55, Mixer_1], properties {inputVolume [1,1]}}
 {id=Modulator_0, children [Mixer_0], properties {volumeMax=1, volumeMin=0.9}}
 {id=Modulator_1, children [QualityLevel_1], properties {volumeMax=0.85, volumeMin=0.8}}
 {id=Modulator_13, children [QualityLevel_2], properties {volumeMin=1}}
 {id=Modulator_2, children [Delay_0], properties {pitchMax=1.1, pitchMin=1}}
 {id=Modulator_4, children [Delay_1], properties {pitchMax=0.9, pitchMin=0.8}}
 {id=QualityLevel_0, children [Random_5, Random_5, WavePlayer_24, WavePlayer_24]}
 {id=QualityLevel_1, children [Random_8, Random_8, WavePlayer_35, WavePlayer_35]}
 {id=QualityLevel_2, children [Random_7, Random_7, WavePlayer_31, WavePlayer_31]}
 {id=Random_4, children [WavePlayer_16, WavePlayer_17, WavePlayer_18], properties {weights [1,1,1]}}
 {id=Random_5, children [WavePlayer_20, WavePlayer_21, WavePlayer_22, WavePlayer_23], properties {weights [1,1,1,1]}}
 {id=Random_6, children [WavePlayer_25, WavePlayer_26, WavePlayer_27], properties {weights [1,1,1]}}
 {id=Random_7, children [WavePlayer_28, WavePlayer_29, WavePlayer_30], properties {weights [1,1,1]}}
 {id=Random_8, children [WavePlayer_32, WavePlayer_33, WavePlayer_34], properties {weights [1,1,1]}}
 {
  id=SoundClass_0
  children [Attenuation_82]
  properties {soundClassOverride=&Fort_WeaponRanged_Fire_Distant}
 }
 {
  id=SoundClass_55
  children [Attenuation_80]
  properties {soundClassOverride=&Fort_WeaponRanged_Fire_3P_PSM}
 }
 {
  id=SoundClass_56
  children [Attenuation_81]
  properties {soundClassOverride=&Fort_WeaponRanged_Fire_Distant}
 }
 {id=WavePlayer_16, properties {soundWaveAssetPtr=/Game/Sounds/Fort_Weapons/SniperRifles/MilitarySniper/MilitarySniper_Casing_01}}
 {id=WavePlayer_17, properties {soundWaveAssetPtr=/Game/Sounds/Fort_Weapons/SniperRifles/MilitarySniper/MilitarySniper_Casing_02}}
 {id=WavePlayer_18, properties {soundWaveAssetPtr=/Game/Sounds/Fort_Weapons/SniperRifles/MilitarySniper/MilitarySniper_Casing_03}}
 {id=WavePlayer_20, properties {soundWaveAssetPtr=&/Turret_Ballistic_Fire_01}}
 {id=WavePlayer_21, properties {soundWaveAssetPtr=&/Turret_Ballistic_Fire_02}}
 {id=WavePlayer_22, properties {soundWaveAssetPtr=&/Turret_Ballistic_Fire_03}}
 {id=WavePlayer_23, properties {soundWaveAssetPtr=&/Turret_Ballistic_Fire_04}}
 {id=WavePlayer_24, properties {soundWaveAssetPtr=&Turret_Ballistic_Fire_01}}
 {id=WavePlayer_25, properties {soundWaveAssetPtr=&MilitarySniper_Casing_01}}
 {id=WavePlayer_26, properties {soundWaveAssetPtr=&MilitarySniper_Casing_02}}
 {id=WavePlayer_27, properties {soundWaveAssetPtr=&MilitarySniper_Casing_03}}
 {id=WavePlayer_28, properties {soundWaveAssetPtr=&/Turret_Ballistic_Fire_Medium_01}}
 {id=WavePlayer_29, properties {soundWaveAssetPtr=&/Turret_Ballistic_Fire_Medium_02}}
 {id=WavePlayer_30, properties {soundWaveAssetPtr=&/Turret_Ballistic_Fire_Medium_03}}
 {id=WavePlayer_31, properties {soundWaveAssetPtr=&Turret_Ballistic_Fire_Medium_01}}
 {id=WavePlayer_32, properties {soundWaveAssetPtr=&/Turret_Ballistic_Fire_Distant_01}}
 {id=WavePlayer_33, properties {soundWaveAssetPtr=&/Turret_Ballistic_Fire_Distant_02}}
 {id=WavePlayer_34, properties {soundWaveAssetPtr=&/Turret_Ballistic_Fire_Distant_03}}
 {id=WavePlayer_35, properties {soundWaveAssetPtr=&Turret_Ballistic_Fire_Distant_01}}
]
properties {
 bOverrideConcurrency=true
 maxDistance=27500
 volumeMultiplier=0.85
 attenuationOverrides {distanceAlgorithm=NaturalSound, falloffDistance=4000, attenuationShapeExtents {x=1000}}
 concurrencyOverrides {maxCount=5}
}
