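# Sound cue for picking up the ballistic mounted turret trap, routed through the weapons sound class.
# A single wave player feeds the output directly, playing the Pickup_01 wave even though the cue is named Pickup_02.
# It plays at reduced volume with weapon-drop attenuation, and its concurrency override blocks new instances at the limit rather than stopping playing ones.
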SoundCue /Game/Sounds/Fort_Traps/MountedTurret/Turret_Ballistic_Pickup_02_Cue
root=WavePlayer_1
nodes [{id=WavePlayer_1, properties {soundWaveAssetPtr=&/Turret_Ballistic_Pickup_01}}]
properties {
 attenuationSettings=&Fort_Weapon_Drop_Att
 bOverrideConcurrency=true
 maxDistance=3000
 soundClassObject=/Game/Sounds/FortSoundClassesAndModes/Fort_Weapons
 volumeMultiplier=0.35
 concurrencyOverrides {resolutionRule=PreventNew}
}
