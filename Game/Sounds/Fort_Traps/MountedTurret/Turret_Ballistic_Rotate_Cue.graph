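# Sound cue for the mounted ballistic turret's rotation, a quiet, attenuated trap sound driven by aim movement.
# A mixer blends two looping turn-aim waves reused from the Ostrich vehicle, with the lower layer weighted louder.
# Continuous modulators scale each layer's pitch and volume: HorizontalAimDelta drives the lower wave, VerticalAimDelta the higher-pitched upper wave.
# Concurrency overrides specify a single instance that blocks new ones.
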SoundCue /Game/Sounds/Fort_Traps/MountedTurret/Turret_Ballistic_Rotate_Cue
root=Mixer_0
nodes [
 {id=Mixer_0, children [ModulatorContinuous_1, ModulatorContinuous_0], properties {inputVolume [1,4]}}
 {
  id=ModulatorContinuous_0
  children [WavePlayer_3]
  properties {
   pitchModulationParams {default=0, minOutput=0.8, parameterName=HorizontalAimDelta}
   volumeModulationParams {default=0, parameterName=HorizontalAimDelta}
  }
 }
 {
  id=ModulatorContinuous_1
  children [WavePlayer_0]
  properties {
   pitchModulationParams {default=0, maxOutput=1.6, minOutput=1.4, parameterName=VerticalAimDelta}
   volumeModulationParams {parameterName=VerticalAimDelta}
  }
 }
 {id=WavePlayer_0, properties {bLooping=true, soundWaveAssetPtr=/Game/Athena/Sounds/Vehicles/Ostrich/Veh_Ostrich_TurnAim_Upper_01}}
 {id=WavePlayer_3, properties {bLooping=true, soundWaveAssetPtr=/Game/Athena/Sounds/Vehicles/Ostrich/Veh_Ostrich_TurnAim_Lower_01}}
]
properties {
 attenuationSettings=&Fort_Weapon_Projectile_Close_Athena_Att
 maxDistance=4950
 soundClassObject=&Fort_Traps
 volumeMultiplier=0.1
 concurrencyOverrides {maxCount=1, resolutionRule=PreventNew}
}
